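# Secondary-motion AnimBP for the MadCommander player body cosmetic, copying its pose from the attached parent mesh.
# Anim Dynamics sways the pipe, armband, grenades and bags, with weighted constraints anchoring the grenades, into a cached pose.
# bIsSkyDiving blends between upper-arm pose drivers (true) and weighted dynamic shoulder-pad constraints (false).
# Look-at and other shoulder constraint chains are unconnected, including the arm-raise alphas computed in the event graph.
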AnimBlueprint /BRCosmetics/Characters/Player/Male/Medium/Bodies/M_MED_MadCommando_01/Meshes/M_MED_MadCommander_AnimBP
target_skeleton=&/M_MED_MadCommander_Skeleton
graphs [
 AnimGraph {nodes [
   n1 = Look At - Bone: pd_shoulderpad_r(Alpha=Get arm up r() -> (arm up r:double)) {
    alphaBoolBlend {blendOption=Linear}
    boneToModify {boneName=pd_shoulderpad_r}
    lookAtLocation {x=0}
    lookAtTarget {boneReference {boneName=lowerarm_r}}
    lookAt_Axis {axis {y=0, z=-1}}
   }
   n2 = Look At - Bone: pd_shoulderpad_l(ComponentPose=@n1, Alpha=Get Arm Up L() -> (Arm Up L:double)) {
    alphaBoolBlend {blendOption=Linear}
    boneToModify {boneName=pd_shoulderpad_l}
    lookAtLocation {x=0}
    lookAtTarget {boneReference {boneName=lowerarm_l}}
    lookAt_Axis {axis {y=0, z=-1}}
   }
   Component To Local(ComponentPose=@n2)
   n3 = "Constraint - pd_shoulderpad_l (PPPP)"!(
    ComponentPose
    Alpha=1
    ConstraintWeights_0=0.1422
    ConstraintWeights_1:float=0.0049
    ConstraintWeights_2:float=0.032
    ConstraintWeights_3:float=0.8208
   ) -> (Pose) {
    alphaBoolBlend {blendOption=Linear}
    boneToModify {boneName=pd_shoulderpad_l}
    constraintSetup [
     {targetBone {boneName=upperarm_twist_01_l}}
     {targetBone {boneName=spine_04}}
     {targetBone {boneName=spine_05}}
     {targetBone {boneName=clavicle_l}}
    ]
    constraintWeights [1,1,1,1]
   }
   "Constraint - pd_shoulderpad_r (PPPPP)"!(
    ComponentPose=@n3
    Alpha=1
    ConstraintWeights_0=0.8185
    ConstraintWeights_1:float=0.0384
    ConstraintWeights_2:float=0.1088
    ConstraintWeights_3:float=0.005
    ConstraintWeights_4:float=0.0292
   ) -> (Pose) {
    alphaBoolBlend {blendOption=Linear}
    boneToModify {boneName=pd_shoulderpad_r}
    constraintSetup [
     {targetBone {boneName=clavicle_r}}
     {targetBone {boneName=upperarm_twist_01_r}}
     {targetBone {boneName=upperarm_r}}
     {targetBone {boneName=spine_04}}
     {targetBone {boneName=spine_05}}
    ]
    constraintWeights [1,1,1,1,1]
   }
   n4 = "Constraint - dyn_shoulderpad_l (PP)"!(ComponentPose, Alpha=1, ConstraintWeights_0=0.2262, ConstraintWeights_1:float=0.7738) -> (Pose) {
    alphaBoolBlend {blendOption=Linear}
    boneToModify {boneName=dyn_shoulderpad_l}
    constraintSetup [{targetBone {boneName=upperarm_twist_01_l}}, {targetBone {boneName=clavicle_l}}]
    constraintWeights [1,1]
   }
   "Constraint - dyn_shoulderpad_r (PP)"!(ComponentPose=@n4, Alpha=1, ConstraintWeights_0=0.7738, ConstraintWeights_1:float=0.2262) -> (Pose) {
    alphaBoolBlend {blendOption=Linear}
    boneToModify {boneName=dyn_shoulderpad_r}
    constraintSetup [{targetBone {boneName=clavicle_r}}, {targetBone {boneName=upperarm_twist_01_r}}]
    constraintWeights [1,1]
   }
   n5 = "Constraint - pd_shoulderpad_l (P)"(Alpha=1, ConstraintWeights_0=1) {
    alphaBoolBlend {blendOption=Linear}
    boneToModify {boneName=pd_shoulderpad_l}
    constraintSetup [{targetBone {boneName=upperarm_twist_01_l}}]
    constraintWeights [1]
   }
   "Constraint - pd_shoulderpad_r (P)"(ComponentPose=@n5, Alpha=1, ConstraintWeights_0=1) {
    alphaBoolBlend {blendOption=Linear}
    boneToModify {boneName=pd_shoulderpad_r}
    constraintSetup [{targetBone {boneName=upperarm_twist_01_r}}]
    constraintWeights [1]
   }
   n6 = "Constraint - dyn_shoulderpad_l (PP)"(ComponentPose=Local To Component(LocalPose=Use cached pose 'CPFM Cache'()), Alpha=1, ConstraintWeights_0=0.7, ConstraintWeights_1=0.3) {
    alphaBoolBlend {blendOption=Linear}
    boneToModify {boneName=dyn_shoulderpad_l}
    constraintSetup [{targetBone {boneName=upperarm_twist_01_l}}, {targetBone {boneName=clavicle_l}}]
    constraintWeights [1,1]
   }
   n7 = "Constraint - dyn_shoulderpad_r (PP)"(ComponentPose=@n6, Alpha=1, ConstraintWeights_0=0.3, ConstraintWeights_1=0.7) {
    alphaBoolBlend {blendOption=Linear}
    boneToModify {boneName=dyn_shoulderpad_r}
    constraintSetup [{targetBone {boneName=clavicle_r}}, {targetBone {boneName=upperarm_twist_01_r}}]
    constraintWeights [1,1]
   }
   n8 = Pose Driver - Source: upperarm_l!(SourcePose:Pose Link Structure=Use cached pose 'CPFM Cache'()) -> (Pose:Pose Link Structure) {
    is=title AnimGraphNode_PoseDriver
    driveSource=Translation
    poseAsset=&/M_MED_MadCommander_PoseAsset
    onlyDriveBones [{boneName=pd_shoulderpad_r}]
    poseTargets [
     {
      bApplyCustomCurve=false
      bIsHidden=false
      distanceMethod=DefaultMethod
      drivenName=Pose_0
      functionType=DefaultFunction
      targetScale=1
      boneTransforms [{targetRotation=<rot -64.7854 -3.089 6.0959>, targetTranslation=<15.5111 0 0>}]
      customCurve {defaultValue=3.4028234663852886e+38, postInfinityExtrap=RCCE_Constant, preInfinityExtrap=RCCE_Constant, keys []}
      targetRotation=<rot 0 0 0>
     }
     {
      bApplyCustomCurve=false
      bIsHidden=false
      distanceMethod=DefaultMethod
      drivenName=Pose_15
      functionType=DefaultFunction
      targetScale=1
      boneTransforms [{targetRotation=<rot 0.1425 2.4363 -82.9397>, targetTranslation=<15.5111 0 0>}]
      customCurve {defaultValue=3.4028234663852886e+38, postInfinityExtrap=RCCE_Constant, preInfinityExtrap=RCCE_Constant, keys []}
      targetRotation=<rot 0 0 0>
     }
     {
      bApplyCustomCurve=false
      bIsHidden=false
      distanceMethod=DefaultMethod
      drivenName=Pose_16
      functionType=DefaultFunction
      targetScale=1
      boneTransforms [{targetRotation=<rot -1.2819 -10.2677 -77.2499>, targetTranslation=<15.5077 -0.0004 -0.0053>}]
      customCurve {defaultValue=3.4028234663852886e+38, postInfinityExtrap=RCCE_Constant, preInfinityExtrap=RCCE_Constant, keys []}
      targetRotation=<rot 0 0 0>
     }
    ]
    rBFParams {distanceMethod=Quaternion, radius=300, weightThreshold=0}
    sourceBones [{boneName=upperarm_l}]
   }
   n9 = Pose Driver - Source: upperarm_l(SourcePose=@n8) {
    poseAsset=&M_MED_MadCommander_PoseAsset
    onlyDriveBones [{boneName=pd_shoulderpad_l}]
    poseTargets [
     {
      bApplyCustomCurve=false
      bIsHidden=false
      distanceMethod=DefaultMethod
      drivenName=Pose_0
      functionType=DefaultFunction
      targetScale=0.164
      boneTransforms [{targetRotation=<rot -64.7854 -3.089 6.0959>, targetTranslation=<15.5111 0 0>}]
      customCurve {defaultValue=3.4028234663852886e+38, postInfinityExtrap=RCCE_Constant, preInfinityExtrap=RCCE_Constant, keys []}
      targetRotation=<rot 0 0 0>
     }
     {
      bApplyCustomCurve=false
      bIsHidden=false
      distanceMethod=DefaultMethod
      drivenName=Pose_15
      functionType=DefaultFunction
      targetScale=0.1478
      boneTransforms [{targetRotation=<rot 0.1425 2.4363 -82.9397>, targetTranslation=<15.5111 0 0>}]
      customCurve {defaultValue=3.4028234663852886e+38, postInfinityExtrap=RCCE_Constant, preInfinityExtrap=RCCE_Constant, keys []}
      targetRotation=<rot 0 0 0>
     }
     {
      bApplyCustomCurve=false
      bIsHidden=false
      distanceMethod=DefaultMethod
      drivenName=Pose_16
      functionType=DefaultFunction
      targetScale=0.1478
      boneTransforms [{targetRotation=<rot -1.2819 -10.2677 -77.2499>, targetTranslation=<15.5077 -0.0004 -0.0053>}]
      customCurve {defaultValue=3.4028234663852886e+38, postInfinityExtrap=RCCE_Constant, preInfinityExtrap=RCCE_Constant, keys []}
      targetRotation=<rot 0 0 0>
     }
    ]
    rBFParams {distanceMethod=Quaternion, radius=300, weightThreshold=0}
    sourceBones [{boneName=upperarm_l}]
   }
   n10 = Pose Driver - Source: upperarm_r!(SourcePose:Pose Link Structure=@n9) -> (Pose:Pose Link Structure) {
    is=title AnimGraphNode_PoseDriver
    driveSource=Translation
    poseAsset=&M_MED_MadCommander_PoseAsset
    onlyDriveBones [{boneName=pd_shoulderpad_r}]
    poseTargets [
     {
      bApplyCustomCurve=false
      bIsHidden=false
      distanceMethod=DefaultMethod
      drivenName=Pose_0
      functionType=DefaultFunction
      targetScale=0.0025
      boneTransforms [{targetRotation=<rot -64.7854 -3.089 6.0959>, targetTranslation=<-15.5111 0 0>}]
      customCurve {defaultValue=3.4028234663852886e+38, postInfinityExtrap=RCCE_Constant, preInfinityExtrap=RCCE_Constant, keys []}
      targetRotation=<rot 0 0 0>
     }
     {
      bApplyCustomCurve=false
      bIsHidden=false
      distanceMethod=DefaultMethod
      drivenName=Pose_15
      functionType=DefaultFunction
      targetScale=0.0025
      boneTransforms [{targetRotation=<rot 0.1425 2.4363 -82.9397>, targetTranslation=<-15.5111 0 0>}]
      customCurve {defaultValue=3.4028234663852886e+38, postInfinityExtrap=RCCE_Constant, preInfinityExtrap=RCCE_Constant, keys []}
      targetRotation=<rot 0 0 0>
     }
     {
      bApplyCustomCurve=false
      bIsHidden=false
      distanceMethod=DefaultMethod
      drivenName=Pose_16
      functionType=DefaultFunction
      targetScale=1
      boneTransforms [{targetRotation=<rot 11.3702 -9.5453 -80.9289>, targetTranslation=<-15.5096 0.0013 0.0035>}]
      customCurve {defaultValue=3.4028234663852886e+38, postInfinityExtrap=RCCE_Constant, preInfinityExtrap=RCCE_Constant, keys []}
      targetRotation=<rot 0 0 0>
     }
    ]
    rBFParams {distanceMethod=Quaternion, radius=300, weightThreshold=0}
    sourceBones [{boneName=upperarm_r}]
   }
   n11 = Pose Driver - Source: upperarm_r(SourcePose=@n10) {
    poseAsset=&M_MED_MadCommander_PoseAsset
    onlyDriveBones [{boneName=pd_shoulderpad_r}]
    poseTargets [
     {
      bApplyCustomCurve=false
      bIsHidden=false
      distanceMethod=DefaultMethod
      drivenName=Pose_0
      functionType=DefaultFunction
      targetScale=0.164
      boneTransforms [{targetRotation=<rot -64.7854 -3.089 6.0959>, targetTranslation=<-15.5111 0 0>}]
      customCurve {defaultValue=3.4028234663852886e+38, postInfinityExtrap=RCCE_Constant, preInfinityExtrap=RCCE_Constant, keys []}
      targetRotation=<rot 0 0 0>
     }
     {
      bApplyCustomCurve=false
      bIsHidden=false
      distanceMethod=DefaultMethod
      drivenName=Pose_15
      functionType=DefaultFunction
      targetScale=0.1768
      boneTransforms [{targetRotation=<rot 0.1425 2.4363 -82.9397>, targetTranslation=<-15.5111 0 0>}]
      customCurve {defaultValue=3.4028234663852886e+38, postInfinityExtrap=RCCE_Constant, preInfinityExtrap=RCCE_Constant, keys []}
      targetRotation=<rot 0 0 0>
     }
     {
      bApplyCustomCurve=false
      bIsHidden=false
      distanceMethod=DefaultMethod
      drivenName=Pose_16
      functionType=DefaultFunction
      targetScale=0.1768
      boneTransforms [{targetRotation=<rot 11.3702 -9.5453 -80.9289>, targetTranslation=<-15.5096 0.0013 0.0035>}]
      customCurve {defaultValue=3.4028234663852886e+38, postInfinityExtrap=RCCE_Constant, preInfinityExtrap=RCCE_Constant, keys []}
      targetRotation=<rot 0 0 0>
     }
    ]
    rBFParams {distanceMethod=Quaternion, radius=300, weightThreshold=0}
    sourceBones [{boneName=upperarm_r}]
   }
   Output Pose(Result=Blend Poses by bool(BlendPose_0=@n11, BlendPose_1=Component To Local(ComponentPose=@n7), bActiveValue=Get bIsSkyDiving()) {blendType=Linear}) {layerGroup=None}
   n12 = Anim Dynamics - Bone: dyn_pipe(ComponentPose=Local To Component(LocalPose=Copy Pose From Mesh() {bUseAttachedParent=true}), Alpha=1, ExternalForce="0,0,0") {
    angularDampingOverride=0.98
    angularSpringConstant=125
    bAngularSpring=true
    bGravityOverrideInSimSpace=true
    bLinearSpring=true
    bOverrideAngularDamping=true
    bOverrideLinearDamping=true
    gravityScale=0.06
    lODThreshold=1
    linearDampingOverride=0.98
    linearSpringConstant=125
    alphaBoolBlend {blendOption=Linear}
    alphaScaleBiasClamp {
     bClampResult=true
     bInterpResult=true
     clampMax=0.25
     clampMin=0.05
     interpSpeedDecreasing=0.5
     interpSpeedIncreasing=0.5
     outRange {max=0.1}
    }
    boundBone {boneName=dyn_pipe}
    physicsBodyDefinitions [{
      sphereCollisionRadius=0
      boundBone {boneName=dyn_pipe}
      boxExtents=<30 30 30>
      constraintSetup {
       angularTargetAxis=AxisY
       twistAxis=AxisZ
       angularLimitsMax=<15 15 15>
       angularLimitsMin=<-15 -15 -15>
       angularTarget=<0 0 0>
       linearAxesMax=<8 2 6>
       linearAxesMin=<-8 -2 -6>
      }
      localJointOffset=<0 0 0>
     }]
   }
   n13 = Anim Dynamics - Bone: dyn_armband_knot(ComponentPose=@n12, Alpha=0.75, ExternalForce="0,0,0") {
    alpha=0.9
    angularDampingOverride=0.98
    angularSpringConstant=1500
    bAngularSpring=true
    bGravityOverrideInSimSpace=true
    bOverrideAngularDamping=true
    numSolverIterationsPostUpdate=4
    numSolverIterationsPreUpdate=16
    simulationSpace=Actor
    alphaBoolBlend {blendOption=Linear}
    boundBone {boneName=dyn_armband_knot}
    physicsBodyDefinitions [{
      sphereCollisionRadius=0
      boundBone {boneName=dyn_armband_knot}
      boxExtents=<15 5 5>
      constraintSetup {angularTargetAxis=AxisZ, twistAxis=AxisY, angularLimitsMax=<15 0 15>, angularLimitsMin=<-15 0 -15>, angularTarget=<0 0 1>}
      localJointOffset=<0 0 0>
     }]
   }
   n14 = Anim Dynamics - Bone: dyn_armband_low(ComponentPose=@n13, Alpha=0.5, ExternalForce="0,0,0") {
    alpha=0.9
    angularDampingOverride=0.98
    angularSpringConstant=1500
    bAngularSpring=true
    bGravityOverrideInSimSpace=true
    bOverrideAngularDamping=true
    numSolverIterationsPostUpdate=4
    numSolverIterationsPreUpdate=16
    simulationSpace=Actor
    alphaBoolBlend {blendOption=Linear}
    boundBone {boneName=dyn_armband_low}
    physicsBodyDefinitions [{
      sphereCollisionRadius=0
      boundBone {boneName=dyn_armband_low}
      boxExtents=<15 5 5>
      constraintSetup {angularTargetAxis=AxisZ, twistAxis=AxisY, angularLimitsMax=<15 15 15>, angularLimitsMin=<-15 -15 -15>, angularTarget=<0 0 1>}
      localJointOffset=<0 0 0>
     }]
   }
   n15 = Anim Dynamics - Bone: dyn_armband_hi(ComponentPose=@n14, Alpha=0.5, ExternalForce="0,0,0") {
    alpha=0.9
    angularDampingOverride=0.98
    angularSpringConstant=1500
    bAngularSpring=true
    bGravityOverrideInSimSpace=true
    bOverrideAngularDamping=true
    numSolverIterationsPostUpdate=4
    numSolverIterationsPreUpdate=16
    simulationSpace=Actor
    alphaBoolBlend {blendOption=Linear}
    boundBone {boneName=dyn_armband_hi}
    physicsBodyDefinitions [{
      sphereCollisionRadius=0
      boundBone {boneName=dyn_armband_hi}
      boxExtents=<15 5 5>
      constraintSetup {angularTargetAxis=AxisZ, twistAxis=AxisY, angularLimitsMax=<15 15 15>, angularLimitsMin=<-15 -15 -15>, angularTarget=<0 0 1>}
      localJointOffset=<0 0 0>
     }]
   }
   n16 = "Constraint - const_grenade_01 (PPPPPPPPPPPP)"!(
    ComponentPose=@n15
    Alpha=1
    ConstraintWeights_0=0
    ConstraintWeights_1:float=0.0011
    ConstraintWeights_2:float=0
    ConstraintWeights_3:float=0
    ConstraintWeights_4:float=0.0072
    ConstraintWeights_5:float=0.166
    ConstraintWeights_6:float=0.3105
    ConstraintWeights_7:float=0.4597
    ConstraintWeights_8:float=0.0217
    ConstraintWeights_9:float=0
    ConstraintWeights_10:float=0.0321
    ConstraintWeights_11:float=0.0015
   ) -> (Pose) {
    alphaBoolBlend {blendOption=Linear}
    boneToModify {boneName=const_grenade_01}
    constraintSetup [
     {targetBone {boneName=upperarm_r}}
     {targetBone {boneName=clavicle_l}}
     {targetBone {boneName=upperarm_twist_01_r}}
     {targetBone {boneName=upperarm_twist_02_r}}
     {targetBone {boneName=spine_01}}
     {targetBone {boneName=spine_02}}
     {targetBone {boneName=spine_03}}
     {targetBone {boneName=spine_04}}
     {targetBone {boneName=spine_05}}
     {targetBone {boneName=upperarm_twist_01_l}}
     {targetBone {boneName=clavicle_r}}
     {targetBone {boneName=pelvis}}
    ]
    constraintWeights [1,1,1,1,1,1,1,1,1,1,1,1]
   }
   n17 = Anim Dynamics - Bone: dyn_grenade_01(ComponentPose=@n16, Alpha=0.35, ExternalForce="0,0,0") {
    angularBiasOverride=1
    angularDampingOverride=0.98
    bGravityOverrideInSimSpace=true
    gravityScale=0.75
    lODThreshold=1
    numSolverIterationsPostUpdate=4
    numSolverIterationsPreUpdate=16
    alphaBoolBlend {blendOption=Linear}
    boundBone {boneName=dyn_grenade_01}
    componentLinearAccScale=<0.2 0.2 0.2>
    componentLinearVelScale=<0.2 0.2 0.2>
    physicsBodyDefinitions [{
      sphereCollisionRadius=0
      boundBone {boneName=dyn_grenade_01}
      boxExtents=<5 2 10>
      constraintSetup {angularTargetAxis=AxisZ, twistAxis=AxisZ, angularLimitsMax=<15 4 15>, angularLimitsMin=<-15 -4 -15>, angularTarget=<0 0 0>}
      localJointOffset=<0 1.1 -1.1>
     }]
   }
   n18 = "Constraint - const_grenade_02 (PPPPPPPPPPPP)"!(
    ComponentPose=@n17
    Alpha=1
    ConstraintWeights_0=0.0002
    ConstraintWeights_1:float=0.0031
    ConstraintWeights_2:float=0.0008
    ConstraintWeights_3:float=0.0001
    ConstraintWeights_4:float=0.0321
    ConstraintWeights_5:float=0.2264
    ConstraintWeights_6:float=0.2984
    ConstraintWeights_7:float=0.2763
    ConstraintWeights_8:float=0.0002
    ConstraintWeights_9:float=0
    ConstraintWeights_10:float=0.1453
    ConstraintWeights_11:float=0.0171
   ) -> (Pose) {
    alphaBoolBlend {blendOption=Linear}
    boneToModify {boneName=const_grenade_02}
    constraintSetup [
     {targetBone {boneName=upperarm_r}}
     {targetBone {boneName=clavicle_l}}
     {targetBone {boneName=upperarm_twist_01_r}}
     {targetBone {boneName=upperarm_twist_02_r}}
     {targetBone {boneName=spine_01}}
     {targetBone {boneName=spine_02}}
     {targetBone {boneName=spine_03}}
     {targetBone {boneName=spine_04}}
     {targetBone {boneName=spine_05}}
     {targetBone {boneName=upperarm_twist_01_l}}
     {targetBone {boneName=clavicle_r}}
     {targetBone {boneName=pelvis}}
    ]
    constraintWeights [1,1,1,1,1,1,1,1,1,1,1,1]
   }
   n19 = Anim Dynamics - Bone: dyn_grenade_02(ComponentPose=@n18, Alpha=0.7, ExternalForce="0,0,0") {
    angularBiasOverride=1
    angularDampingOverride=0.98
    bGravityOverrideInSimSpace=true
    gravityScale=0.25
    lODThreshold=1
    numSolverIterationsPostUpdate=4
    numSolverIterationsPreUpdate=16
    alphaBoolBlend {blendOption=Linear}
    boundBone {boneName=dyn_grenade_02}
    componentLinearAccScale=<0.5 0.5 5>
    componentLinearVelScale=<0.5 0.5 0.5>
    physicsBodyDefinitions [{
      sphereCollisionRadius=0
      boundBone {boneName=dyn_grenade_02}
      boxExtents=<5 2 10>
      constraintSetup {angularTargetAxis=AxisZ, twistAxis=AxisZ, angularLimitsMax=<8 4 8>, angularLimitsMin=<-6 -1 -6>, angularTarget=<0 0 0>}
      localJointOffset=<0 0.5 -0.5>
     }]
   }
   n20 = Anim Dynamics - Bone: dyn_r_bag_01(ComponentPose=@n19, Alpha=1, ExternalForce="0,0,0") {
    bGravityOverrideInSimSpace=true
    gravityScale=0.1
    lODThreshold=1
    alphaBoolBlend {blendOption=Linear}
    alphaScaleBiasClamp {bClampResult=true, clampMax=0.33, clampMin=0.1, scale=0.0033}
    boundBone {boneName=dyn_r_bag_01}
    physicsBodyDefinitions [{
      sphereCollisionRadius=0
      boundBone {boneName=dyn_r_bag_01}
      boxExtents=<5 5 5>
      constraintSetup {angularTargetAxis=AxisY, twistAxis=AxisY, angularLimitsMax=<0 0 0>, angularLimitsMin=<0 0 0>, angularTarget=<0 0 0>, linearAxesMax=<0 0 2>}
      localJointOffset=<0 0 5>
     }]
   }
   n21 = Anim Dynamics - Bone: dyn_l_bag_01(ComponentPose=@n20, Alpha=1, ExternalForce="0,0,0") {
    bGravityOverrideInSimSpace=true
    gravityScale=0.1
    lODThreshold=1
    alphaBoolBlend {blendOption=Linear}
    alphaScaleBiasClamp {bClampResult=true, clampMax=0.33, clampMin=0.1, scale=0.0033}
    boundBone {boneName=dyn_l_bag_01}
    physicsBodyDefinitions [{
      sphereCollisionRadius=0
      boundBone {boneName=dyn_l_bag_01}
      boxExtents=<5 5 5>
      constraintSetup {angularTargetAxis=AxisY, twistAxis=AxisY, angularLimitsMax=<0 0 0>, angularLimitsMin=<0 0 0>, angularTarget=<0 0 0>, linearAxesMax=<0 0 2>}
      localJointOffset=<0 0 5>
     }]
   }
   Save cached pose 'CPFM Cache'(Pose=Component To Local(ComponentPose=@n21)) {cacheName=CPFM Cache}
  ]}
 EventGraph {nodes [
   Event BlueprintUpdateAnimation()
   n1 = GetOwnerSkeletalMeshComponent()
   n2 = BreakTransform(InTransform=GetDeltaTransformFromRefPose(self=@n1, BoneName=upperarm_r, BaseName=clavicle_r))
   > Set arm up r(arm up r:double=MapRangeClamped(Value=BreakRotator(InRot=@n2.Rotation).Yaw, InRangeA=30, InRangeB=65)) -> (Output_Get:double)
   n3 = BreakTransform(InTransform=GetDeltaTransformFromRefPose(self=@n1, BoneName=upperarm_l, BaseName=clavicle_l))
   > Set Arm Up L(Arm Up L:double=MapRangeClamped(Value=BreakRotator(InRot=@n3.Rotation).Yaw, InRangeA=30, InRangeB=65)) -> (Output_Get:double)
  ]}
]
variables [
 Arm Up L {type=number}
 arm up r {type=number}
]
